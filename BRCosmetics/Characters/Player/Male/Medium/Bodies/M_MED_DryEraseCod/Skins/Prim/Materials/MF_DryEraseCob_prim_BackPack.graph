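# Material function for the DryEraseCob backpack cosmetic that sets an animated, audio-reactive emissive glow on incoming material attributes.
# FadingIn scales frontend-loop audio amplitude, combined with FadingOut by max and shaped by an exponent.
# A packed lighting mask supplies gradient, emissive and detail channels, alongside panning noise for a light-flow animation.
# The glow blends three-color gradient and two-color Fresnel lighting, with TechArtMask's blue channel also feeding the output.
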MaterialFunction /BRCosmetics/Characters/Player/Male/Medium/Bodies/M_MED_DryEraseCod/Skins/Prim/Materials/MF_DryEraseCob_prim_BackPack
graph {nodes [
  n1 = GetMaterialAttributes(None=FunctionInput() {inputType=FunctionInput_MaterialAttributes}) {attributeGetTypes [#1]}
  n2 = LinearInterpolate(Alpha=NamedRerouteUsage()) {constB=2}
  n3 = NamedRerouteUsage()
  n4 = TextureSampleParameter2D() {
   group=05_Tech Art Masks
   parameterName=TechArtMask
   samplerType=SAMPLERTYPE_Masks
   sortPriority=0
   texture=&T_BlackMask
  }
  n5 = Multiply(A=@n4.B, B=NamedRerouteUsage())
  n6 = Constant() {r=1}
  n7 = LinearInterpolate(B=NamedRerouteUsage(), Alpha=Max(A=Power(Base=@n3) {constExponent=1}, B=Multiply(A=@n5, B=@n6))) {constB=0}
  FunctionOutput(None=SetMaterialAttributes!(MaterialAttributes=@n1.MaterialAttributes, BaseColor=@n1.MaterialAttributes, EmissiveColor=Multiply(A=@n2, B=@n7)) {attributeSetTypes [#1, #2]})
  n8 = ScalarParameter() {group=DryEraseCob, parameterName=FadingIn}
  n9 = Multiply(
   A=MaterialFunctionCall!(InAmplitude, InSpectrum, CustomFakeSpectrum, CustomFakeAmplitude) {materialFunction=/BRCosmetics/Global/MaterialFunctions/AudioReactivity/MF_AudioReactivity_FrontendLoop}.Amplitude
   B=ScalarParameter() {defaultValue=1, group=DryEraseCob, parameterName=EnableAudioReactivity, sliderMax=1}
  )
  NamedRerouteDeclaration(None=Max(A=ScalarParameter() {group=DryEraseCob, parameterName=FadingOut}, B=Multiply(A=@n8, B=@n9))) {name=Amplitude, nodeColor=<rgba 0 1 0.1094 1>}
  n10 = TextureSampleParameter2D() {
   group=DryEraseCob
   parameterName=LightingMask
   samplerType=SAMPLERTYPE_Masks
   texture=/BRCosmetics/Accessories/FORT_Backpacks/Backpack_M_MED_DryEraseCod/Skin/DryEraseCod_Prim/Textures/T_DryEraseCod_Prim_Backpack_Fx2
  }
  NamedRerouteDeclaration(None=@n10.R) {name=Gradient mask_R, nodeColor=<rgba 0.8516 0 1 1>}
  NamedRerouteDeclaration(None=@n10.B) {name=DetailMask, nodeColor=<rgba 0.2656 1 0 1>}
  NamedRerouteDeclaration(None=Power(Base=NamedRerouteUsage(), Exp=ScalarParameter() {defaultValue=1.5, group=DryEraseCob, parameterName=AmplitudeEmissiveExp}) {constExponent=8}) {name=AmplitudeAdjusted, nodeColor=<rgba 0 1 0.9531 1>}
  n11 = Saturate(None=Fresnel(
    ExponentIn=ScalarParameter() {defaultValue=2.3965, group=DryEraseCob, parameterName=Fresnel_Exponentln}
    BaseReflectFractionIn=ScalarParameter() {defaultValue=0.06, group=DryEraseCob, parameterName=Fresnel_BaseReflectFractionln}
   ))
  n12 = Constant() {r=2.5}
  NamedRerouteDeclaration(None=Saturate(None=Power(Base=@n11, Exp=@n12))) {name=Fresnel lighting, nodeColor=<rgba 0.7813 0 1 1>}
  n13 = Panner(Coordinate=Multiply(A=TextureCoordinate(), B=ScalarParameter() {defaultValue=2, group=DryEraseCob, parameterName=PannerNoiseTiling}) {constB=2}) {speedY=0.1}
  NamedRerouteDeclaration(None=Multiply(
    A=TextureSample(UVs=@n13) {samplerType=SAMPLERTYPE_Grayscale, texture=/Game/Global/Textures/Default/Noise/LowResBlurredNoiseNormalized}.R
    B=Constant() {r=1}
   ) {constB=0.35}) {name=PannerNoise, nodeColor=<rgba 0 1 0.4609 1>}
  n14 = LinearInterpolate(Alpha=NamedRerouteUsage()) {constA=0.05, constB=0.99}
  n15 = NamedRerouteUsage()
  n16 = Saturate(None=Add(A=Saturate(None=@n14), B=LinearInterpolate(Alpha=@n15) {constA=-0.6}))
  NamedRerouteDeclaration(None=Multiply(A=NamedRerouteDeclaration(None=@n16) {name=Animation, nodeColor=<rgba 1 0.6094 0 1>}, B=NamedRerouteUsage())) {name=Light flow animation_1, nodeColor=<rgba 1 0.9375 0 1>}
  n17 = NamedRerouteUsage()
  n18 = LinearInterpolate(
   A=VectorParameter() {group=DryEraseCob, parameterName=FresnelColorB, defaultValue=<rgba 1 0.284 0 1>}.RGB
   B=VectorParameter() {group=DryEraseCob, parameterName=FresnelColorA, defaultValue=<rgba 3 0.9948 0 1>}.RGB
   Alpha=NamedRerouteUsage()
  )
  n19 = NamedRerouteUsage()
  n20 = Multiply(
   A=MaterialFunctionCall(
    A=VectorParameter() {group=DryEraseCob, parameterName=ColorA, defaultValue=<rgba 2 0 0.0453 1>}.RGB
    B=VectorParameter() {group=DryEraseCob, parameterName=ColorB, defaultValue=<rgba 1.5 0.3385 0 1>}.RGB
    C=VectorParameter() {group=DryEraseCob, parameterName=ColorC, defaultValue=<rgba 4 1.3264 0 1>}.RGB
    Alpha=NamedRerouteUsage()
   ) {materialFunction=&Lerp_3Color}.Result
   B=OneMinus(None=NamedRerouteUsage())
  )
  NamedRerouteDeclaration(None=Add(A=Multiply(A=@n17, B=@n18), B=Multiply(A=@n19, B=@n20))) {name=Emissive ouput, nodeColor=<rgba 1 0.375 0 1>}
  NamedRerouteDeclaration(None=NamedRerouteDeclaration(None=@n10.G) {name=Emissive_Mask, nodeColor=<rgba 0.0547 0 1 1>}) {name=FX Mask Input, nodeColor=<rgba 0.0547 0 1 1>}
 ]}
