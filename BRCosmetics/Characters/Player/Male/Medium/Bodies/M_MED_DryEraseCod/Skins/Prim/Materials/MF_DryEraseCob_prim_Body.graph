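# Material function for the DryEraseCob Prim body skin that layers an animated, audio-reactive orange-red glow onto incoming material attributes, overriding BaseColor and EmissiveColor.
# An amplitude signal combines FadingOut with FadingIn scaled by frontend-loop audio reactivity.
# Emissive, lighting and tech-art textures supply FX masks, while panned noise, time-driven sine and Fresnel terms provide flowing light and rim effects.
# Named reroutes link these stages, including three-color gradient and Fresnel color blends for emissive output.
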MaterialFunction /BRCosmetics/Characters/Player/Male/Medium/Bodies/M_MED_DryEraseCod/Skins/Prim/Materials/MF_DryEraseCob_prim_Body
graph {nodes [
  n1 = GetMaterialAttributes(None=FunctionInput() {inputType=FunctionInput_MaterialAttributes}) {attributeGetTypes [#1]}
  n2 = Constant3Vector() {constant=<rgba 0.8632 0.4072 0.1441 1>}
  n3 = Multiply(A=NamedRerouteUsage(), B=NamedRerouteUsage())
  n4 = LinearInterpolate(Alpha=NamedRerouteUsage()) {constB=2}
  n5 = LinearInterpolate(B=NamedRerouteUsage(), Alpha=Power(Base=NamedRerouteUsage()) {constExponent=1}) {constB=0}
  FunctionOutput(None=SetMaterialAttributes!(
    MaterialAttributes=@n1.MaterialAttributes
    BaseColor=LinearInterpolate(A=@n1.BaseColor, B=@n2, Alpha=@n3)
    EmissiveColor=Multiply(A=@n4, B=@n5)
   ) {attributeSetTypes [#1, #2]})
  n6 = ScalarParameter() {group=DryEraseCob, parameterName=FadingIn}
  n7 = Multiply(
   A=MaterialFunctionCall!(InAmplitude, InSpectrum, CustomFakeSpectrum, CustomFakeAmplitude) {materialFunction=/BRCosmetics/Global/MaterialFunctions/AudioReactivity/MF_AudioReactivity_FrontendLoop}.Amplitude
   B=ScalarParameter() {defaultValue=1, group=DryEraseCob, parameterName=EnableAudioReactivity, sliderMax=1}
  )
  NamedRerouteDeclaration(None=Max(A=ScalarParameter() {group=DryEraseCob, parameterName=FadingOut}, B=Multiply(A=@n6, B=@n7))) {name=Amplitude, nodeColor=<rgba 0 1 0.1094 1>}
  n8 = TextureSampleParameter2D() {
   group=DryEraseCob
   parameterName=Emissive_Mask
   samplerType=SAMPLERTYPE_Masks
   texture=/BRCosmetics/Characters/Player/Male/Medium/Bodies/M_MED_DryEraseCod/Skins/Prim/Textures/T_DryEraseCod_Prim_Body_FX2
  }
  NamedRerouteDeclaration(None=@n8.G) {name=Detail Mask, nodeColor=<rgba 0.0547 0 1 1>}
  n9 = TextureSampleParameter2D() {
   group=DryEraseCob
   parameterName=LightingMask
   samplerType=SAMPLERTYPE_Masks
   texture=/BRCosmetics/Characters/Player/Male/Medium/Bodies/M_MED_DryEraseCod/Skins/Prim/Textures/T_DryEraseCod_Prim_Body_FX3
  }
  NamedRerouteDeclaration(None=@n9.R) {name=Gradient mask_R, nodeColor=<rgba 0.8516 0 1 1>}
  NamedRerouteDeclaration(None=@n9.B) {name=MaskAroundCloth_B, nodeColor=<rgba 0.6875 0 1 1>}
  n10 = LinearInterpolate(Alpha=NamedRerouteUsage()) {constA=0.01, constB=0.99}
  n11 = NamedRerouteUsage()
  n12 = Saturate(None=Add(A=Saturate(None=@n10), B=LinearInterpolate(Alpha=@n11) {constA=-0.6, constB=0.7}))
  NamedRerouteDeclaration(None=@n12) {name=Animation, nodeColor=<rgba 1 0.6094 0 1>}
  n13 = Saturate(None=Fresnel(
    ExponentIn=ScalarParameter() {defaultValue=2.3965, group=DryEraseCob, parameterName=Fresnel_Exponentln}
    BaseReflectFractionIn=ScalarParameter() {defaultValue=0.06, group=DryEraseCob, parameterName=Fresnel_BaseReflectFractionln}
   ))
  n14 = Constant() {r=1.5}
  NamedRerouteDeclaration(None=Saturate(None=Power(Base=@n13, Exp=@n14))) {name=Fresnel lighting, nodeColor=<rgba 0.7813 0 1 1>}
  NamedRerouteDeclaration(None=Power(Base=NamedRerouteUsage(), Exp=ScalarParameter() {defaultValue=1.5, group=DryEraseCob, parameterName=AmplitudeEmissiveExp}) {constExponent=8}) {name=AmplitudeAdjusted, nodeColor=<rgba 0 1 0.9531 1>}
  n15 = Panner(Coordinate=Multiply(A=TextureCoordinate(), B=ScalarParameter() {defaultValue=2, group=DryEraseCob, parameterName=PannerNoiseTiling}) {constB=2}) {speedY=0.15}
  NamedRerouteDeclaration(None=Multiply(
    A=TextureSample(UVs=@n15) {samplerType=SAMPLERTYPE_Grayscale, texture=/Game/Global/Textures/Default/Noise/LowResBlurredNoiseNormalized}.R
    B=Constant() {r=1}
   ) {constB=0.35}) {name=PannerNoise, nodeColor=<rgba 0 1 0.4609 1>}
  n16 = TextureSampleParameter2D() {
   group=DryEraseCob
   parameterName=Emissive_Mask
   samplerType=SAMPLERTYPE_Masks
   texture=&T_DryEraseCod_Prim_Body_FX2
  }
  n17 = NamedRerouteUsage()
  n18 = NamedRerouteUsage()
  NamedRerouteDeclaration(None=Max(A=Multiply(A=@n12, B=@n16.R), B=LinearInterpolate(B=@n17, Alpha=@n18))) {name=Light flow animation_1, nodeColor=<rgba 1 0.9375 0 1>}
  n19 = Time()
  n20 = Constant() {r=0.45}
  n21 = Constant() {r=-2}
  n22 = Constant() {r=1}
  n23 = Saturate(None=NamedRerouteUsage())
  n24 = MaterialFunctionCall(Value=Add(A=Multiply(A=@n19, B=@n20) {constB=0.5}, B=LinearInterpolate(A=@n21, B=@n22, Alpha=@n23) {constB=0.8})) {materialFunction=&LinearSine}
  NamedRerouteDeclaration(None=Power(Base=Multiply(A=@n24.Linear Sine)) {constExponent=4}) {name=Lighting animation2, nodeColor=<rgba 0 1 0.6953 1>}
  n25 = NamedRerouteUsage()
  n26 = NamedRerouteUsage()
  n27 = MaterialFunctionCall(
   A=VectorParameter() {group=DryEraseCob, parameterName=ColorA, defaultValue=<rgba 3 0 0.068 1>}.RGB
   B=VectorParameter() {group=DryEraseCob, parameterName=ColorB, defaultValue=<rgba 1.5 0.3385 0 1>}.RGB
   C=VectorParameter() {group=DryEraseCob, parameterName=ColorC, defaultValue=<rgba 6 1.9896 0 1>}.RGB
   Alpha=@n26
  ) {materialFunction=&Lerp_3Color}
  n28 = LinearInterpolate(
   B=ScalarParameter() {defaultValue=10, group=DryEraseCob, parameterName=luminous point brightness}
   Alpha=Multiply(A=@n26, B=NamedRerouteUsage())
  ) {constA=2}
  n29 = Multiply(
   A=LinearInterpolate(A=@n27.Result, B=Multiply(A=@n27.Result, B=@n28) {constB=2}, Alpha=NamedRerouteUsage())
   B=LinearInterpolate(Alpha=OneMinus(None=@n25)) {constA=0.05}
  )
  n30 = LinearInterpolate(
   A=VectorParameter() {group=DryEraseCob, parameterName=FresnelColorB, defaultValue=<rgba 1 0.284 0 1>}.RGB
   B=VectorParameter() {group=DryEraseCob, parameterName=FresnelColorA, defaultValue=<rgba 3 0.9948 0 1>}.RGB
   Alpha=NamedRerouteUsage()
  )
  n31 = TextureSampleParameter2D() {
   group=DryEraseCob
   parameterName=Emissive_Mask
   samplerType=SAMPLERTYPE_Masks
   texture=&T_DryEraseCod_Prim_Body_FX2
  }
  n32 = Add(A=NamedRerouteUsage(), B=@n31.B)
  n33 = Constant3Vector() {constant=<rgba 1.2 0.0995 0 1>}
  n34 = Constant3Vector() {constant=<rgba 15 5.2921 0 1>}
  n35 = NamedRerouteUsage()
  n36 = Saturate(None=Multiply(A=NamedRerouteUsage()) {constB=1.5})
  n37 = NamedRerouteUsage()
  NamedRerouteDeclaration(None=LinearInterpolate(A=LinearInterpolate(A=@n29, B=@n30, Alpha=@n32), B=LinearInterpolate(A=@n33, B=@n34, Alpha=@n35), Alpha=Multiply(A=@n36, B=@n37))) {name=Emissive ouput, nodeColor=<rgba 1 0.375 0 1>}
  n38 = NamedRerouteDeclaration(None=@n9.G) {name=Luminous point_G, nodeColor=<rgba 0 1 0.5313 1>}
  n39 = Max(A=@n8.R, B=Multiply(A=Power(Base=@n8.B), B=ScalarParameter() {defaultValue=0.25, group=DryEraseCob, parameterName=Edge brightness}))
  n40 = TextureSampleParameter2D() {
   group=05_Tech Art Masks
   parameterName=TechArtMask
   samplerType=SAMPLERTYPE_Masks
   sortPriority=0
   texture=&T_BlackMask
  }
  n41 = Max(A=@n40.R, B=@n40.R)
  n42 = LinearInterpolate(A=@n41, B=OneMinus(None=@n41), Alpha=ScalarParameter() {group=DryEraseCob, parameterName=Hat?})
  n43 = NamedRerouteUsage()
  NamedRerouteDeclaration(None=Max(A=Max(A=@n38, B=@n39), B=Multiply(A=@n42, B=@n43))) {name=FX Mask Input, nodeColor=<rgba 0.0547 0 1 1>}
 ]}
