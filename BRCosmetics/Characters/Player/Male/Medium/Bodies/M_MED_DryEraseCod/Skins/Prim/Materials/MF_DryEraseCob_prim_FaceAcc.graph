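# Material function for the DryEraseCob Prim skin's face accessory that adds an audio-reactive emissive effect to incoming material attributes.
# Fresnel, panning noise and mask textures drive blends between two color pairs to build the glow.
# A TechArtMask, invertible through a "Hat?" parameter, feeds the FX mask.
# An Amplitude signal combines frontend-loop audio reactivity with an enable scalar and FadingIn/FadingOut parameters.
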MaterialFunction /BRCosmetics/Characters/Player/Male/Medium/Bodies/M_MED_DryEraseCod/Skins/Prim/Materials/MF_DryEraseCob_prim_FaceAcc
graph {nodes [
  n1 = GetMaterialAttributes(None=FunctionInput() {inputType=FunctionInput_MaterialAttributes}) {attributeGetTypes [#1]}
  n2 = NamedRerouteUsage()
  n3 = LinearInterpolate(B=NamedRerouteUsage(), Alpha=Power(Base=NamedRerouteUsage()) {constExponent=1}) {constB=0}
  FunctionOutput(None=SetMaterialAttributes!(MaterialAttributes=@n1.MaterialAttributes, BaseColor=@n1.MaterialAttributes, EmissiveColor=Multiply(A=@n2, B=@n3)) {attributeSetTypes [#1, #2]})
  MaterialFunctionCall() {materialFunction=/Game/Packages/Fortress_SharedMaterials/Base_Material_Functions/BaseCharacterShaderFunctions/Audio/MF_AudioReactivity}
  n4 = ScalarParameter() {group=DryEraseCob, parameterName=FadingIn}
  n5 = Multiply(
   A=MaterialFunctionCall!(InAmplitude, InSpectrum, CustomFakeSpectrum, CustomFakeAmplitude) {materialFunction=/BRCosmetics/Global/MaterialFunctions/AudioReactivity/MF_AudioReactivity_FrontendLoop}.Amplitude
   B=ScalarParameter() {defaultValue=1, group=DryEraseCob, parameterName=EnableAudioReactivity, sliderMax=1}
  )
  NamedRerouteDeclaration(None=Max(A=ScalarParameter() {group=DryEraseCob, parameterName=FadingOut}, B=Multiply(A=@n4, B=@n5))) {name=Amplitude, nodeColor=<rgba 0 1 0.1094 1>}
  n6 = TextureSampleParameter2D() {
   group=00_Texture Inputs
   parameterName=Mask
   samplerType=SAMPLERTYPE_Masks
   sortPriority=3
   texture=&T_BlackMask
  }
  NamedRerouteDeclaration(None=@n6.R) {name=Detail Mask, nodeColor=<rgba 0.0547 0 1 1>}
  n7 = Saturate(None=Fresnel(
    ExponentIn=ScalarParameter() {defaultValue=2.3965, group=DryEraseCob, parameterName=Fresnel_Exponentln}
    BaseReflectFractionIn=ScalarParameter() {defaultValue=0.06, group=DryEraseCob, parameterName=Fresnel_BaseReflectFractionln}
   ))
  n8 = Constant() {r=1}
  NamedRerouteDeclaration(None=Saturate(None=Power(Base=@n7, Exp=@n8))) {name=Fresnel lighting, nodeColor=<rgba 0.7813 0 1 1>}
  n9 = TextureCoordinate()
  n10 = ScalarParameter() {defaultValue=0.98, parameterName=Scale}
  n11 = Panner(Coordinate=Multiply(
    A=MaterialFunctionCall(UVs=@n9, Texture Scale=@n10) {materialFunction=&ScaleUVsByCenter}.UVs
    B=ScalarParameter() {defaultValue=2, group=DryEraseCob, parameterName=PannerNoiseTiling}
   ) {constB=2}) {speedX=0.15}
  NamedRerouteDeclaration(None=Multiply(
    A=TextureSample(UVs=@n11) {samplerType=SAMPLERTYPE_Grayscale, texture=/Game/Global/Textures/Default/Noise/LowResBlurredNoiseNormalized}.R
    B=Constant() {r=1}
   ) {constB=0.35}) {name=PannerNoise, nodeColor=<rgba 0 1 0.4609 1>}
  n12 = Constant()
  n13 = VectorParameter() {group=DryEraseCob, parameterName=FresnelColorB, defaultValue=<rgba 1 0.284 0 1>}
  n14 = VectorParameter() {group=DryEraseCob, parameterName=FresnelColorA, defaultValue=<rgba 3 0.9948 0 1>}
  n15 = NamedRerouteUsage()
  n16 = VectorParameter() {group=DryEraseCob, parameterName=FresnelColorB_2, defaultValue=<rgba 2 0.0006 0.0783 1>}
  n17 = VectorParameter() {group=DryEraseCob, parameterName="FresnelColorA_@", defaultValue=<rgba 20 6.6321 0 1>}
  n18 = NamedRerouteUsage()
  n19 = TextureSample() {
   samplerType=SAMPLERTYPE_Masks
   texture=/BRCosmetics/Characters/Player/Male/Medium/Bodies/M_MED_DryEraseCod/Skins/Prim/Textures/T_DryEraseCod_Prim_FaceAcc_FX2
  }
  n20 = Power(Base=@n19.R) {constExponent=1}
  n21 = LinearInterpolate(
   A=Multiply(A=LinearInterpolate(A=@n13.RGB, B=@n14.RGB, Alpha=@n15)) {constB=0.8}
   B=Multiply(A=LinearInterpolate(A=@n16.RGB, B=@n17.RGB, Alpha=@n18)) {constB=2}
   Alpha=@n20
  )
  n22 = Max(A=NamedRerouteUsage(), B=@n20)
  n23 = OneMinus(None=NamedRerouteUsage())
  NamedRerouteDeclaration(None=Multiply(A=LinearInterpolate(A=@n12, B=@n21, Alpha=@n22), B=LinearInterpolate(Alpha=@n23) {constA=0.05})) {name=Emissive ouput, nodeColor=<rgba 1 0.375 0 1>}
  n24 = TextureSampleParameter2D() {
   group=05_Tech Art Masks
   parameterName=TechArtMask
   samplerType=SAMPLERTYPE_Masks
   sortPriority=0
   texture=&T_BlackMask
  }
  n25 = Max(A=@n24.R, B=@n24.R)
  n26 = OneMinus(None=@n25)
  n27 = ScalarParameter() {group=DryEraseCob, parameterName=Hat?}
  NamedRerouteDeclaration(None=Multiply(A=LinearInterpolate(A=@n25, B=@n26, Alpha=@n27), B=NamedRerouteUsage())) {name=FX Mask Input, nodeColor=<rgba 0.0547 0 1 1>}
 ]}
