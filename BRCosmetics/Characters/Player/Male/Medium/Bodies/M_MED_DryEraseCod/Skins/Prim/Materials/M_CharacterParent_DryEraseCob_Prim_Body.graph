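# Parent character material for the Prim skin of the DryEraseCob cosmetic, serving body, hat and backpack variants on skeletal meshes.
# It wraps the shared 2023 character PreFX and PostFX functions around a dry-erase thin-film effect driven by a Fresnel term.
# An IsHat static switch selects face accessory or body logic, then IsBackpack optionally layers backpack logic on that result.
# It uses material attributes and selects the shading model by expression.
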Material /BRCosmetics/Characters/Player/Male/Medium/Bodies/M_MED_DryEraseCod/Skins/Prim/Materials/M_CharacterParent_DryEraseCob_Prim_Body
graph {
 properties {
  bUseMaterialAttributes=true
  bUsedWithMeshDeformer=true
  bUsedWithMorphTargets=true
  bUsedWithSkeletalMesh=true
  pixelDepthOffsetMode=PDOM_Legacy
  shadingModel=MSM_FromMaterialExpression
  usedShadingModels=Default Lit | Subsurface | Clear Coat | Subsurface Profile | Cloth
 }
 nodes [
  n1 = MaterialFunctionCall() {materialFunction=&MF_CharacterParent_2023_PreFX}
  n2 = MaterialFunctionCall!(Material Attributes=@n1.Result, Fresnel=Fresnel(ExponentIn=Constant() {r=1}) {baseReflectFraction=0}, Alpha=@n1.Result) {materialFunction=&/MF_ThinFilm_TechArt_DryErase}
  n3 = StaticSwitchParameter(
   True=MaterialFunctionCall!(In=@n2.Result) {materialFunction=&/MF_DryEraseCob_prim_FaceAcc}.Result
   False=MaterialFunctionCall!(In=@n2.Result) {materialFunction=&/MF_DryEraseCob_prim_Body}.Result
  ) {group=DryEraseCob, parameterName=IsHat}
  MaterialFunctionCall(
   Material=StaticSwitchParameter(
    True=MaterialFunctionCall!(In=@n3) {materialFunction=&/MF_DryEraseCob_prim_BackPack}.Result
    False=@n3
   ) {group=DryEraseCob, parameterName=IsBackpack}
   M=@n1.M
   Pre Skinned Local Position=@n1.M
   Distance Blend=@n1.M
   Fresnel=@n1.M
  ) {materialFunction=&MF_CharacterParent_2023_PostFX}
 ]
}
